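; Asset Data Utility Tool modifier that, when enabled, prints a configurable message about an asset data array.
; RunArrayModifier replaces %c with the array count, then calls GenerateAssetListString if the message contains %ls.
; That helper swaps %ls(N) for newline-separated asset names, with positive N capping the last index, inclusively.
; The list step starts from the raw OutputMessage, so it discards the %c substitution.
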(fn GenerateAssetListString
  (params (Message :type String) (Data :type Array%20of%20Asset%20Data%20Structures))
  (bind _l_startindex (LocalVariables|Getl_startIndex))
  (bind _returnvalue (+ _l_startindex 4))
  (bind _l_message (LocalVariables|Getl_message))
  (LocalVariables|Setl_message Message)
  (LocalVariables|Setl_data Data)
  (LocalVariables|Setl_lastIndex (Utilities|Array|LastIndex (LocalVariables|Getl_data)))
  (Utilities|FlowControl|Sequence
    (:then_0
      (if (Utilities|String|Contains (LocalVariables|Getl_message) "%ls(")
        (LocalVariables|Setl_startIndex (Utilities|String|FindSubstring _l_message "%ls("))
        (LocalVariables|Setl_endIndex (Utilities|String|FindSubstring _l_message ")" false false _returnvalue))
        (bind _output_get (LocalVariables|Setl_foundLastIndex (Utilities|String|StringToInteger (Utilities|String|GetSubstring (LocalVariables|Getl_message) _returnvalue (- (LocalVariables|Getl_endIndex) _returnvalue)))))
        (if (> _output_get 0)
          (LocalVariables|Setl_lastIndex (Math|Integer|Min(Integer) (LocalVariables|Getl_foundLastIndex) (LocalVariables|Getl_lastIndex))))))
    (:then_1
      (bind _index (Utilities|FlowControl|ForLoopwithBreak 0 (LocalVariables|Getl_lastIndex))
        (:LoopBody
          (LocalVariables|Setl_return (Utilities|String|Append (LocalVariables|Getl_return) (select (== _index 0) "" "
") (Utilities|String|ToString(Name) (Utilities|Struct|BreakAssetData (Utilities|Array|Get(acopy) (LocalVariables|Getl_data) _index))))))))
    (:then_2
      (return (Utilities|String|Replace (LocalVariables|Getl_message) (Utilities|String|GetSubstring (LocalVariables|Getl_message) (LocalVariables|Getl_startIndex) (Utilities|Operators|Subtract (LocalVariables|Getl_endIndex) (LocalVariables|Getl_startIndex) -1)) (LocalVariables|Getl_return))))))

(fn RunArrayModifier
  (params (InOutAssetDataArray :type Array%20of%20Asset%20Data%20Structures))
  (if (Variables|Default|GetEnabled)
    (LocalVariables|Setl_data InOutAssetDataArray)
    (LocalVariables|Setl_count (Utilities|Array|Length InOutAssetDataArray))
    (Utilities|FlowControl|Sequence
      (:then_0
        (LocalVariables|Setl_output (Utilities|String|Replace (Variables|Default|GetOutputMessage) "%c" (Utilities|String|ToString(Integer) (LocalVariables|Getl_count)))))
      (:then_1
        (if (Utilities|String|Contains (Variables|Default|GetOutputMessage) "%ls")
          (bind _newmessage (CallFunction|GenerateAssetListString (Variables|Default|GetOutputMessage) (LocalVariables|Getl_data)))
          (LocalVariables|Setl_output _newmessage)))
      (:then_2
        (Development|PrintString 0 (LocalVariables|Getl_output))))))
